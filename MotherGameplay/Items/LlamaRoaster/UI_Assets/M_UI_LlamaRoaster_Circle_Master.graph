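# Translucent UI master material for the LlamaRoaster circle graphic, drawing a procedural disc or ring.
# Exponential radial gradients form the shape, with the IsRing static switch cutting out the center.
# UseMask blends in a sharpened UV.y cutoff that by default keeps the top half, and Final_Alpha scales the resulting opacity.
# The shape also blends Falloff_Color to Filled_Color, alongside IsErrror-driven lerps of each color toward Error_Color.
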Material /MotherGameplay/Items/LlamaRoaster/UI_Assets/M_UI_LlamaRoaster_Circle_Master
graph {
 properties {blendMode=BLEND_Translucent, materialDomain=MD_UI, pixelDepthOffsetMode=PDOM_Legacy, usedShadingModels=Default Lit}
 nodes [
  n1 = VectorParameter() {group=Color, parameterName=Falloff_Color, defaultValue=<rgba 1 0 0 1>}
  n2 = VectorParameter() {group=Color, parameterName=Filled_Color, defaultValue=<rgba 0 1 0 1>}
  n3 = ScalarParameter() {defaultValue=0.4, group=Treatment, parameterName=Inside_Radius}
  n4 = ScalarParameter() {defaultValue=16, group=Treatment, parameterName=Inside_Density}
  n5 = Saturate(None=OneMinus(None=MaterialFunctionCall(Radius=@n3, Density=@n4) {materialFunction=&RadialGradientExponential}.RadialGradientExponential))
  n6 = Saturate(None=MaterialFunctionCall(
    Radius=ScalarParameter() {defaultValue=0.5, group=Treatment, parameterName=Outside_Radius}
    Density=ScalarParameter() {defaultValue=16, group=Treatment, parameterName=Outside_Density}
   ) {materialFunction=&RadialGradientExponential}.RadialGradientExponential)
  n7 = StaticSwitchParameter(True=Saturate(None=DotProduct(A=@n5, B=@n6)), False=@n6) {group=Global, parameterName=IsRing}
  n8 = ComponentMask(None=Add(A=TextureCoordinate(), B=ScalarParameter() {defaultValue=-0.5, parameterName=OffsetMask})) {g=true}
  n9 = ScalarParameter() {defaultValue=64, parameterName=MaskEdge}
  n10 = OneMinus(None=Saturate(None=Multiply(A=@n8, B=@n9) {constB=64}))
  n11 = LinearInterpolate(A=@n7, B=Saturate(None=DotProduct(A=@n7, B=@n10)), Alpha=ScalarParameter() {parameterName=UseMask})
  LinearInterpolate(A=@n1.RGB, B=@n2.RGB, Alpha=@n11)
  n12 = VectorParameter() {group=Color, parameterName=Error_Color, defaultValue=<rgba 0 0 1 1>}
  LinearInterpolate(A=@n2.RGB, B=@n12.RGB, Alpha=ScalarParameter() {group=Color, parameterName=IsErrror})
  LinearInterpolate(A=@n1.RGB, B=@n12.RGB, Alpha=ScalarParameter() {group=Color, parameterName=IsErrror})
  Saturate(None=Multiply(A=@n11, B=ScalarParameter() {defaultValue=1, group=Global, parameterName=Final_Alpha}))
 ]
}
